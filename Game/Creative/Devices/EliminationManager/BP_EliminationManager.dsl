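; Creative Elimination Manager device that reacts to filtered player and AI eliminations by dropping items, broadcasting elimination events and optionally applying player penalties.
; On minigame start, the server binds player deaths through an added controller component and AI deaths through the population tracker, optionally awaiting despawn.
; Eligible deaths passing a drop-chance roll queue locations that the spawner component consumes sequentially.
; Penalties remove, drop or transfer resources and score; minigame end resets everything.
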(event Custom|CleanUpMinigameDelegates
  (bind _fortminigameprogress (Variables|Default|GetFortMinigameProgress)
    (:then
      (bind _currentminigame (Variables|/Script/FortniteGame.FortMinigameLogicComponent|GetCurrentMinigame _fortminigameprogress)
        (:then
          (Minigame|UnbindEventfromPlayerRemovedDelegate _currentminigame (EventDispatchers|CreateEvent :!event Custom|OnPlayerRemoved))
          (Minigame|UnbindEventfromOnPlayerJoininProgress _currentminigame (EventDispatchers|CreateEvent :!event Custom|OnPlayerJoinedInProgress))
          (Minigame|UnbindEventfromOnAddedtoMinigame _fortminigameprogress (EventDispatchers|CreateEvent :!event OnAddedToMinigame))
          (bind _returnvalue (AI|GetAIPopulationTracker))
          (AI|UnbindEventfromOnAIPawnDied _returnvalue (EventDispatchers|CreateEvent :!event Custom|OnAIPawnDied)))
        (:else)))
    (:else)))

(event Custom|OnAIPawnDied
  (params (KilledController :type Controller%20Object%20Reference) (bIsABot :type Boolean) (KilledAIPawn :type Fort%20Pawn%20Object%20Reference) (KilledPlayerPawn :type Fort%20Player%20Pawn%20Object%20Reference) (KillerController :type Controller%20Object%20Reference))
  (Utilities|IsValid KilledPlayerPawn
    (:Is%20Valid
      (if bIsABot
        (CallFunction|HandlePawnDeath KilledPlayerPawn KillerController)))
    (:Is%20Not%20Valid
      (bind _output_get (Variables|Default|SetKilledAIPawn KilledAIPawn))
      (bind _output_get_2 (Variables|Default|SetKillerController KillerController))
      (if (not (Variables|Default|GetWaitUntilEliminatedDespawns))
        (CallFunction|HandlePawnDeath _output_get _output_get_2)))))

(event Custom|On%20Pawn%20Destroyed
  (params (DestroyedPawn :type Fort%20Pawn%20Object%20Reference))
  (if (Variables|Default|GetWaitUntilEliminatedDespawns)
    (CallFunction|HandlePawnDeath (Variables|Default|GetKilledAIPawn) (Variables|Default|GetKillerController))))

(event EventHandleOnMinigameStarted
  (Networking|SwitchHasAuthority
    (:Authority
      (Variables|Default|GetMinigame_Spawner_Component
        (:then
          (CallFunction|SetShouldTossPickups)
          (Variables|/Game/Athena/Items/Traps/MinigameSpawner/Minigame_Spawner_Component.Minigame_Spawner_Component_C|SetRandomSpawnOption (Variables|Default|GetRandomDrop) (Variables|Default|GetMinigame_Spawner_Component))
          (Variables|/Game/Athena/Items/Traps/MinigameSpawner/Minigame_Spawner_Component.Minigame_Spawner_Component_C|SetMaxSpawns (Variables|Default|GetNumberOfItemsToDrop) (Variables|Default|GetMinigame_Spawner_Component))
          (bind _fortminigameprogress (Variables|Default|GetFortMinigameProgress)
            (:then
              (bind _currentminigame (Variables|/Script/FortniteGame.FortMinigameLogicComponent|GetCurrentMinigame _fortminigameprogress)
                (:then
                  (Utilities|FlowControl|Sequence
                    (:then_0
                      (for _array_element (Minigame|GetParticipatingPlayers _currentminigame)
                        (Utilities|IsValid _array_element
                          (:Is%20Valid
                            (Utilities|FlowControl|Sequence
                              (:then_0
                                (Utilities|IsValid (Fortnite|GetControllerfromActor _array_element)
                                  (:Is%20Valid
                                    (Pawn|BindEventtoOnPossessedPawnChanged (Fortnite|GetControllerfromActor _array_element) (EventDispatchers|CreateEvent :!event Custom|OnPlayerPawnPossessed)))))
                              (:then_1
                                (CallFunction|BindtoPawnDeath (PlayerState|GetCurrentPawn _array_element))))))))
                    (:then_1
                      (Minigame|AssignOnPlayerJoininProgress _currentminigame (EventDispatchers|CreateEvent :!event Custom|OnPlayerJoinedInProgress)))
                    (:then_2
                      (Minigame|AssignPlayerRemovedDelegate _currentminigame (EventDispatchers|CreateEvent :!event Custom|OnPlayerRemoved)))
                    (:then_3
                      (DeviceOptions|CustomInitializeTeams (Variables|DeviceOptions|GetTeamSetting))
                      (DeviceOptions|CustomInitializeClass (Variables|DeviceOptions|GetClassIndex)))
                    (:then_4
                      (bind _returnvalue (AI|GetAIPopulationTracker))
                      (AI|BindEventtoOnAIPawnDied _returnvalue (EventDispatchers|CreateEvent :!event Custom|OnAIPawnDied)))
                    (:then_5
                      (bind _returnvalue_2 (AI|GetAIPopulationTracker))
                      (AI|BindEventtoOnAIPawnDestroyed _returnvalue_2 (EventDispatchers|CreateEvent :!event Custom|On%20Pawn%20Destroyed)))))
                (:else)))
            (:else)))
        (:else)))))

(event Custom|OnPlayerJoinedInProgress
  (params (Pawn :type Pawn%20Object%20Reference))
  (bind _returnvalue (Pawn|GetController Pawn))
  (Utilities|FlowControl|Sequence
    (:then_0
      (CallFunction|BindtoPawnDeath Pawn))
    (:then_1
      (Utilities|IsValid _returnvalue
        (:Is%20Valid
          (Pawn|BindEventtoOnPossessedPawnChanged _returnvalue (EventDispatchers|CreateEvent :!event Custom|OnPlayerPawnPossessed)))))))

(event EventHandleMinigameEnded
  (Networking|SwitchHasAuthority
    (:Authority
      (CallFunction|Reset))))

(event Custom|OnPlayerPawnPossessed
  (params (OldPawn :type Pawn%20Object%20Reference) (NewPawn :type Pawn%20Object%20Reference))
  (CallFunction|BindtoPawnDeath NewPawn))

(event EventEndPlay
  (params (EndPlayReason :type EEndPlayReason%20Enum :default "Destroyed"))
  (|CallParentFunction EndPlayReason)
  (Networking|SwitchHasAuthority
    (:Authority
      (CallFunction|Reset))))

(event OnLastItemFinishedSpawning(Minigame_Spawner_Component)
  (Networking|SwitchHasAuthority
    (:Authority
      (if (> (Utilities|Array|Length (Variables|Default|GetQueuedSpawnLocations)) 0)
        (Utilities|Array|RemoveIndex (Variables|Default|GetQueuedSpawnLocations))
        (bind _minigame_spawner_component (Variables|Default|GetMinigame_Spawner_Component)
          (:then
            (Class|MinigameSpawnerComponent|ResetSpawner _minigame_spawner_component false)
            (if (> (Utilities|Array|Length (Variables|Default|GetQueuedSpawnLocations)) 0)
              (Class|MinigameSpawnerComponent|SpawnRemaining (Variables|Default|GetMinigame_Spawner_Component) true)))
          (:else))))))

(event EventHandleSpawnItemWhenReceived
  (Networking|SwitchHasAuthority
    (:Authority
      (if (not (Variables|Default|GetbDisabledSpawning))
        (bind _fortminigameprogress (Variables|Default|GetFortMinigameProgress)
          (:then
            (bind _currentminigame (Variables|/Script/FortniteGame.FortMinigameLogicComponent|GetCurrentMinigame _fortminigameprogress)
              (:then
                (for _array_element (Minigame|GetParticipatingPlayers _currentminigame)
                  (Utilities|IsValid _array_element
                    (:Is%20Valid
                      (bind _valid_controller (DEPRECATED|ActorHasValidController (Variables|Default|GetCreative_UseRestriction_Component) (PlayerState|GetCurrentPawn _array_element) true true))
                      (if _valid_controller
                        (CallFunction|QueuePlayerLocation (PlayerState|GetCurrentPawn _array_element)))))))
              (:else)))
          (:else))))))

(event OnChannelIdsReplicated(WhenEliminated)
  (CallFunction|UpdateSpawnerFX))

(event OnChannelIdsReplicated(OnElimination)
  (CallFunction|UpdateSpawnerFX))

(event EventBeginPlay
  (|CallParentFunction)
  (CallFunction|SetupScreenFX)
  (Variables|Default|SetCachedMinigameStatsComponent (Minigame|Stats|GetMinigameStatsComponent self)))

(event EventUpdateSpawnerFX
  ;[1 1 1] Insert FX nodes here
  (|CallParentFunction)
  (CallFunction|UpdateScreenColor))

(event Custom|CleanUpPlayerDelegates
  (params (PlayerState :type Fort%20Player%20State%20Object%20Reference))
  (bind _returnvalue (Fortnite|GetControllerfromActor PlayerState))
  (Utilities|IsValid PlayerState
    (:Is%20Valid
      (Utilities|FlowControl|Sequence
        (:then_0
          (Utilities|IsValid _returnvalue
            (:Is%20Valid
              (Pawn|UnbindEventfromOnPossessedPawnChanged _returnvalue (EventDispatchers|CreateEvent :!event Custom|OnPlayerPawnPossessed)))))
        (:then_1
          (CallFunction|BindtoPawnDeath (CallFunction|GetPlayerPawnAthenafromPlayerState PlayerState)))))))

(event OnAddedtoMinigame(FortMinigameProgress)
  (params (Minigame :type Fort%20Minigame%20Object%20Reference))
  (Networking|SwitchHasAuthority
    (:Authority
      ;[1 1 1] Setup FortMinigameProgress to enable ClassRestrictions
      (Initialization|InitializeComponent (Variables|Default|GetCreative_UseRestriction_Component) (Variables|Default|GetFortMinigameProgress) 0 0 (Knobs|Team|GetShiftedAnyTeamValue) -1 0 1 :!class /Game/Creative/Devices/Common/Components/Creative_UseRestriction_Component.Creative_UseRestriction_Component_C))))

(event Custom|OnPlayerRemoved
  (params (UniqueNetId :type Unique%20Net%20Id%20Structure) (bIsLocalPlayer :type Boolean))
  (CallFunction|CleanUpPlayerDelegates (CallFunction|GetFortPlayerStatebyUniqueNetId UniqueNetId)))

(fn BindToPawnDeath
  (params (Pawn :type Pawn%20Object%20Reference))
  (bind _asfort_player_pawn_athena (Utilities|Casting|CastToFortPlayerPawnAthena Pawn)
    (:then
      (bind _returnvalue (Components|AddControllerComponent (Pawn|GetController _asfort_player_pawn_athena) "/Game/Creative/Devices/EliminationManager/BP_FortControllerComponent_EliminationManagement.BP_FortControllerComponent_EliminationManagement_C"))
      (bind _asbp_fort_controller_component_elimination_management (Utilities|Casting|CastToBP_FortControllerComponent_EliminationManagement _returnvalue)
        (:then
          (Default|BindEventtoOnPawnDeath _asbp_fort_controller_component_elimination_management (EventDispatchers|CreateEvent :!event OnPawnDeath))
          (return true))
        (:CastFailed
          (return false))))
    (:CastFailed
      (return false))))

(fn CleanUpPlayers
  (bind _fortminigameprogress (Variables|Default|GetFortMinigameProgress)
    (:then
      (bind _currentminigame (Variables|/Script/FortniteGame.FortMinigameLogicComponent|GetCurrentMinigame _fortminigameprogress)
        (:then
          (for _array_element (Minigame|GetParticipatingPlayers _currentminigame)
            (CallFunction|CleanUpPlayerDelegates _array_element)))
        (:else)))
    (:else)))

(fn CustomInitializeClass
  (params (SelectedClass :type Integer))
  (Networking|SwitchHasAuthority
    (:Authority
      (ClassRestrictions|SetClassWithAffects (Variables|Default|GetCreative_UseRestriction_Component) SelectedClass (== (Variables|DeviceOptions|GetKnob_EliminatedAffectsClass) (Variables|/Game/Creative/Devices/Common/Components/Creative_UseRestriction_Component.Creative_UseRestriction_Component_C|GetAffectsAllButSelectedValue (Variables|Default|GetCreative_UseRestriction_Component))) false))))

(fn CustomInitializeTeams
  (params (SelectedTeam :type Integer))
  (Networking|SwitchHasAuthority
    (:Authority
      (TeamRestrictions|SetTeamswithAffects (Variables|Default|GetCreative_UseRestriction_Component) SelectedTeam (== (Variables|DeviceOptions|GetKnob_EliminatedAffectsTeam) (Variables|/Game/Creative/Devices/Common/Components/Creative_UseRestriction_Component.Creative_UseRestriction_Component_C|GetAffectsAllButSelectedValue (Variables|Default|GetCreative_UseRestriction_Component)))))))

(fn DroporRemoveMaterials
  (params (Eliminated%20Controller :type Controller%20Object%20Reference) (DropParams :type Fort%20Creative%20Item%20Drop%20Params%20Structure))
  (bind _itemremovedcount (Inventory|DropSpecificItem :self (Variables|Default|GetFortCreativeInventoryManip) :Controller (Variables|GetEliminated%20Controller) :ItemDefinition "/Game/Items/ResourcePickups/WoodItemData.WoodItemData" :DropParams (Variables|GetDropParams) :!class /Script/FortniteGame.FortCreativeInventoryManipComponent))
  (bind _itemremovedcount_2 (Inventory|DropSpecificItem :self (Variables|Default|GetFortCreativeInventoryManip) :Controller (Variables|GetEliminated%20Controller) :ItemDefinition "/Game/Items/ResourcePickups/StoneItemData.StoneItemData" :DropParams (Variables|GetDropParams) :!class /Script/FortniteGame.FortCreativeInventoryManipComponent))
  (bind _itemremovedcount_3 (Inventory|DropSpecificItem :self (Variables|Default|GetFortCreativeInventoryManip) :Controller (Variables|GetEliminated%20Controller) :ItemDefinition "/Game/Items/ResourcePickups/MetalItemData.MetalItemData" :DropParams (Variables|GetDropParams) :!class /Script/FortniteGame.FortCreativeInventoryManipComponent))
  (return _itemremovedcount_3 _itemremovedcount_2 _itemremovedcount))

(fn GetBaseSpawnLocation
  (if (> (Utilities|Array|Length (Variables|Default|GetQueuedSpawnLocations)) 0)
    (return (Utilities|Array|Get(acopy) (Variables|Default|GetQueuedSpawnLocations)))
    (else
      (Fortnite|FortShippingLog self "Elimination Manager tried to spawn an item with no queued location." true))))

(fn GetCollisionStaticMesh
  (return "0, 0, 0" (Variables|/Script/Engine.StaticMeshComponent|GetStaticMesh (Variables|Debug|GetStaticMeshComponent))))

(fn GetFinisherOrDownerPawn
  (params (DyingPlayerPawn :type Fort%20Player%20Pawn%20Athena%20Object%20Reference))
  (bind _playerstate (Variables|/Script/Engine.Pawn|GetPlayerState DyingPlayerPawn)
    (:then
      (bind _asfort_player_state_athena (Utilities|Casting|CastToFortPlayerStateAthena _playerstate))
      (bind _returnvalue_finisherordowner (Utilities|Struct|BreakDeathInfo (Class|FortPlayerStateAthena|GetDeathInfo _asfort_player_state_athena)))
      (Utilities|IsValid _returnvalue_finisherordowner
        (:Is%20Valid
          (Utilities|IsValid (Fortnite|GetControllerfromActor _returnvalue_finisherordowner)
            (:Is%20Valid
              (bind _asfort_player_pawn_athena (Utilities|Casting|CastToFortPlayerPawnAthena (Pawn|GetControlledPawn (Fortnite|GetControllerfromActor _returnvalue_finisherordowner))))
              (return _asfort_player_pawn_athena))))))
    (:else)))

(fn GetFortPlayerStateByUniqueNetId
  (params (UniqueNetId :type Unique%20Net%20Id%20Structure))
  (bind _returnvalue (Game|GetGameState))
  (Utilities|FlowControl|Sequence
    (:then_0
      (Utilities|IsValid _returnvalue
        (:Is%20Valid
          (for _array_element (Variables|/Script/Engine.GameStateBase|GetPlayerArray _returnvalue)
            (bind _asfort_player_state_athena (Utilities|Casting|CastToFortPlayerStateAthena _array_element))
            (if (Online|UniqueNetIdRepl|Equal(UniqueIdRepl) (PlayerState|GetPlayerUniqueNetID _asfort_player_state_athena) UniqueNetId)
              (LocalVariables|SetPlayerState _asfort_player_state_athena))))))
    (:then_1
      (return (LocalVariables|GetPlayerState)))))

(fn GetMeshComponents
  (bind _returnvalue (Class|CreativeMoveToolOverrides|GetMeshComponents(Message) (Variables|Mesh|GetStaticMesh)))
  (return _returnvalue))

(fn GetPlayerPawnAthenaFromPlayerState
  (params (PlayerState :type Fort%20Player%20State%20Object%20Reference))
  (Utilities|FlowControl|Sequence
    (:then_0
      (Utilities|IsValid PlayerState
        (:Is%20Valid
          (bind _asfort_player_pawn_athena (Utilities|Casting|CastToFortPlayerPawnAthena (PlayerState|GetCurrentPawn PlayerState)))
          (LocalVariables|SetFortPlayerPawnAthena _asfort_player_pawn_athena))))
    (:then_1
      (return (LocalVariables|GetFortPlayerPawnAthena)))))

(fn GetVisibilityComponentsWithCollision
  (return (Utilities|Map|MakeMap 0 "NoCollision" (Variables|Default|GetPreviewStaticMesh) "NoCollision" (Variables|Default|GetPreviewSkeletalMesh) "NoCollision")))

(fn GrantItemsToPlayer
  (params (InventoryOwner :type Fort%20Inventory%20Owner%20Interface%20Interface) (GoldAmount :type Integer) (MetalAmount :type Integer) (StoneAmount :type Integer) (WoodAmount :type Integer))
  (Inventory|GiveItemtoInventoryOwner :InventoryOwner InventoryOwner :ItemDefinition "/Game/Items/ResourcePickups/GoldItemData.GoldItemData" :NumberToGive GoldAmount :bNotifyPlayer true :ItemLevel -1 :PickupInstigatorHandle -1 :bUseItemPickupAnalyticEvent false :WeaponAmmoOverride -1)
  (Inventory|GiveItemtoInventoryOwner :InventoryOwner InventoryOwner :ItemDefinition "/Game/Items/ResourcePickups/MetalItemData.MetalItemData" :NumberToGive MetalAmount :bNotifyPlayer true :ItemLevel -1 :PickupInstigatorHandle -1 :bUseItemPickupAnalyticEvent false :WeaponAmmoOverride -1)
  (Inventory|GiveItemtoInventoryOwner :InventoryOwner InventoryOwner :ItemDefinition "/Game/Items/ResourcePickups/StoneItemData.StoneItemData" :NumberToGive StoneAmount :bNotifyPlayer true :ItemLevel -1 :PickupInstigatorHandle -1 :bUseItemPickupAnalyticEvent false :WeaponAmmoOverride -1)
  (Inventory|GiveItemtoInventoryOwner :InventoryOwner InventoryOwner :ItemDefinition "/Game/Items/ResourcePickups/WoodItemData.WoodItemData" :NumberToGive WoodAmount :bNotifyPlayer true :ItemLevel -1 :PickupInstigatorHandle -1 :bUseItemPickupAnalyticEvent false :WeaponAmmoOverride -1))

(fn HandleEliminationPenalties
  (params (Eliminated%20Controller :type Controller%20Object%20Reference) (Eliminator%20Controller :type Controller%20Object%20Reference))
  (bind _penaltyamount (Variables|DeviceOptions|GetPenaltyAmount))
  (bind _eliminatedplayerscore (LocalVariables|GetEliminatedPlayerScore))
  (bind _returnvalue (* _eliminatedplayerscore _penaltyamount))
  (bind _returnvalue_2 (Math|Float|Truncate _returnvalue))
  (bind _removedgold (LocalVariables|GetRemovedGold))
  (bind _removedmetal (LocalVariables|GetRemovedMetal))
  (bind _removedstone (LocalVariables|GetRemovedStone))
  (bind _removedwood (LocalVariables|GetRemovedWood))
  (bind _eliminator_controller (Variables|GetEliminator%20Controller))
  (bind _cachedminigamestatscomponent (Variables|Default|GetCachedMinigameStatsComponent)
    (:then
      (LocalVariables|SetEliminatedPlayerScore (Minigame|Stats|GetStatforPlayer _cachedminigamestatscomponent (Utilities|Casting|CastToFortPlayerState (Variables|/Script/Engine.Controller|GetPlayerState Eliminated%20Controller)) (Class|FortMinigameStatsComponent|GetLoadedStatfromClass _cachedminigamestatscomponent "/Game/Athena/Playset/Minigames/StatFilters/MinigameStat_Score.MinigameStat_Score_C") false))
      (switch Utilities|FlowControl|Switch|SwitchonPenaltyEffectOption (Variables|DeviceOptions|GetPenaltyEffect)
        (:NewEnumerator0
          (LocalVariables|SetDropParams (Utilities|Struct|MakeFortCreativeItemDropParams "Percentage" 0 (Variables|DeviceOptions|GetPenaltyAmount) false)))
        (:NewEnumerator1
          (LocalVariables|SetDropParams (Utilities|Struct|MakeFortCreativeItemDropParams "Percentage" 0 (Variables|DeviceOptions|GetPenaltyAmount))))
        (:NewEnumerator2
          (LocalVariables|SetTransferItems true)
          (LocalVariables|SetDropParams (Utilities|Struct|MakeFortCreativeItemDropParams "Percentage" 0 (Variables|DeviceOptions|GetPenaltyAmount) false))))
      (Utilities|FlowControl|Sequence
        (:then_0
          (switch Utilities|FlowControl|Switch|SwitchonPenaltyItemOption (Variables|DeviceOptions|GetPenaltyItemOption)
            (:NewEnumerator0
              (Utilities|FlowControl|Sequence
                (:then_0
                  _)
                (:then_1
                  (bind (_removedmetal_2 _removedstone_2 _removedwood_2) (CallFunction|DropOrRemoveMaterials (Variables|GetEliminated%20Controller) (LocalVariables|GetDropParams)))
                  (LocalVariables|SetRemovedMetal _removedmetal_2)
                  (LocalVariables|SetRemovedStone _removedstone_2)
                  (LocalVariables|SetRemovedWood _removedwood_2))
                (:then_2
                  (LocalVariables|SetRemovedScore _returnvalue_2)
                  (Class|CreativeCommonDeviceFunctionLibrary|AwardScore self _returnvalue_2 "NewEnumerator2" (Variables|GetEliminated%20Controller) (Variables|Default|GetFortMinigameProgress)))))
            (:NewEnumerator1
              _)
            (:NewEnumerator2
              _)
            (:NewEnumerator3
              _))
          (bind _itemremovedcount (Inventory|DropSpecificItem :self (Variables|Default|GetFortCreativeInventoryManip) :Controller (Variables|GetEliminated%20Controller) :ItemDefinition "/Game/Items/ResourcePickups/GoldItemData.GoldItemData" :DropParams (LocalVariables|GetDropParams) :!class /Script/FortniteGame.FortCreativeInventoryManipComponent))
          (LocalVariables|SetRemovedGold _itemremovedcount))
        (:then_1
          (if (LocalVariables|GetTransferItems)
            (Utilities|IsValid (Variables|GetEliminator%20Controller)
              (:Is%20Valid
                (Class|CreativeCommonDeviceFunctionLibrary|AwardScore self (LocalVariables|GetRemovedScore) "NewEnumerator1" (Variables|GetEliminator%20Controller) (Variables|Default|GetFortMinigameProgress))
                (bind _asfort_player_controller (Utilities|Casting|CastToFortPlayerController _eliminator_controller)
                  (:then
                    (CallFunction|GrantItemstoPlayer _asfort_player_controller _removedgold _removedmetal _removedstone _removedwood))
                  (:CastFailed
                    (bind _asfort_athena_aibot_controller (Utilities|Casting|CastToFortAthenaAIBotController _eliminator_controller))
                    (CallFunction|GrantItemstoPlayer _asfort_athena_aibot_controller _removedgold _removedmetal _removedstone _removedwood)))))))))
    (:else)))

(fn HandlePawnDeath
  (params (Killed%20AI%20Pawn :type Fort%20Pawn%20Object%20Reference) (Killer%20Controller :type Controller%20Object%20Reference))
  (bind _enemytypeoption (Variables|Default|GetEnemyTypeOption))
  (bind _guardtypeoptionmap (Variables|Default|GetGuardTypeOptionMap))
  (bind (_value _returnvalue) (Utilities|Map|Find _guardtypeoptionmap _enemytypeoption))
  (Utilities|FlowControl|Sequence
    (:then_0
      (Utilities|IsValid Killed%20AI%20Pawn
        (:Is%20Valid
          (if (not (CallFunction|IsParticipatingPlayer Killed%20AI%20Pawn))
            (if (Utilities|Enum|Equal(Enum) _enemytypeoption "NewEnumerator0")
              (LocalVariables|SetIsKilledValid true)
              (elif (Utilities|Enum|Equal(Enum) _enemytypeoption "NewEnumerator1")
                (LocalVariables|SetIsKilledValid false) ; Players are Handled in Handle Player Died
                (elif (Utilities|Enum|Equal(Enum) _enemytypeoption "NewEnumerator3")
                  (LocalVariables|SetIsKilledValid true)
                  (elif (Utilities|Enum|Equal(Enum) _enemytypeoption "NewEnumerator30")
                    (if (GameplayTags|HasAnyTags (GameplayTags|MakeGameplayTagContainerfromArray (Variables|Default|GetGuardGameplayTags)) (GameplayTags|GetOwnedGameplayTags (Variables|GetKilled%20AI%20Pawn) :!class /Script/GameplayTags.GameplayTagAssetInterface) true)
                      (LocalVariables|SetIsKilledValid true))
                    (elif _returnvalue
                      (if (GameplayTags|HasTag (GameplayTags|GetOwnedGameplayTags (Variables|GetKilled%20AI%20Pawn) :!class /Script/GameplayTags.GameplayTagAssetInterface) _value true)
                        (LocalVariables|SetIsKilledValid true))
                      (elif (or (GameplayTags|HasTag (GameplayTags|GetOwnedGameplayTags (Variables|GetKilled%20AI%20Pawn) :!class /Script/GameplayTags.GameplayTagAssetInterface) (Utilities|Map|Find (Variables|Default|GetEnemyTypeOptionMap) _enemytypeoption) true) (GameplayTags|HasTag (Variables|/Script/FortniteGame.FortPawn|GetGameplayTags (Variables|GetKilled%20AI%20Pawn)) (Utilities|Map|Find (Variables|Default|GetEnemyTypeOptionMap) _enemytypeoption) true))
                        (LocalVariables|SetIsKilledValid true)))))))))))
    (:then_1
      (if (LocalVariables|GetIsKilledValid)
        (CallFunction|QueuePlayerLocation (Variables|GetKilled%20AI%20Pawn))
        (GameplayEvent|BroadcastEvent (Variables|UserOptions|Events|GetOn%20Eliminated) 0 (Utilities|Struct|MakeGameplayEventwhenElimination (Pawn|GetController (Variables|GetKilled%20AI%20Pawn))))
        (Utilities|IsValid (Variables|GetKiller%20Controller)
          (:Is%20Valid
            (GameplayEvent|BroadcastEvent (Variables|UserOptions|Events|GetOn%20Elimination) 0 (Utilities|Struct|MakeGameplayEventonElimination (Variables|GetKiller%20Controller))))
          (:Is%20Not%20Valid
            (GameplayEvent|BroadcastEvent (Variables|UserOptions|Events|GetOn%20Elimination) 0 (Utilities|Struct|MakeGameplayEventonElimination 0))))))))

(fn HandlePlayerDied
  (params (FinisherOrDownerPawn :type Fort%20Player%20Pawn%20Athena%20Object%20Reference) (DyingPlayerPawn :type Fort%20Player%20Pawn%20Athena%20Object%20Reference%20%28by%20ref%29))
  (if (or (Utilities|Enum|Equal(Enum) (Variables|Default|GetEnemyTypeOption) "NewEnumerator0") (Utilities|Enum|Equal(Enum) (Variables|Default|GetEnemyTypeOption) "NewEnumerator1"))
    (bind _valid_controller (DEPRECATED|ActorHasValidController (Variables|Default|GetCreative_UseRestriction_Component) DyingPlayerPawn true (and (not (Variables|Default|GetbDisabledSpawning)) (CallFunction|PassesSelfEliminationTest DyingPlayerPawn))))
    (if _valid_controller
      (CallFunction|QueuePlayerLocation (Variables|GetDyingPlayerPawn))
      (GameplayEvent|BroadcastEvent (Variables|UserOptions|Events|GetOn%20Eliminated) 0 (Utilities|Struct|MakeGameplayEventwhenElimination (Pawn|GetController (Variables|GetDyingPlayerPawn))))
      (Utilities|IsValid (Variables|GetFinisherOrDownerPawn)
        (:Is%20Valid
          (GameplayEvent|BroadcastEvent (Variables|UserOptions|Events|GetOn%20Elimination) 0 (Utilities|Struct|MakeGameplayEventonElimination (Pawn|GetController (Variables|GetFinisherOrDownerPawn))))
          (if (Variables|DeviceOptions|GetEliminationPenalty)
            (CallFunction|HandleEliminationPenalties (Pawn|GetController (Variables|GetDyingPlayerPawn)) (Pawn|GetController (Variables|GetFinisherOrDownerPawn)))))
        (:Is%20Not%20Valid
          (GameplayEvent|BroadcastEvent (Variables|UserOptions|Events|GetOn%20Elimination) 0 (Utilities|Struct|MakeGameplayEventonElimination 0))
          (if (Variables|DeviceOptions|GetEliminationPenalty)
            (CallFunction|HandleEliminationPenalties (Pawn|GetController (Variables|GetDyingPlayerPawn)))))))))

(fn HasChannelSet
  (bind _wheneliminated (Variables|Default|GetWhenEliminated)
    (:then
      (bind _onelimination (Variables|Default|GetOnElimination)
        (:then
          (return (or (Class|FortGameplayMessageComponentBase|IsChannelIdSet _onelimination) (Class|FortGameplayMessageComponentBase|IsChannelIdSet _wheneliminated))))
        (:else)))
    (:else)))

(fn HasItems
  (bind _minigame_spawner_component (Variables|Default|GetMinigame_Spawner_Component)
    (:then
      (return (> (Utilities|Array|Length (Variables|/Script/FortniteGame.FortMinigameSpawnerComponent|GetToSpawnList _minigame_spawner_component)) 0)))
    (:else)))

(fn IgnoreWorldGridScaleOverride
  (return false))

(fn IsParticipatingPlayer
  (params (Pawn :type Pawn%20Object%20Reference))
  ;[1 0.9385 0.3582] Returns whether the given pawn is a participating player in the game. Used to avoid tracking player bots as both AI and player eliminations
  (Utilities|FlowControl|Sequence
    (:then_0
      (Utilities|IsValid Pawn
        (:Is%20Valid
          (bind _asfort_player_state_athena (Utilities|Casting|CastToFortPlayerStateAthena (Variables|/Script/Engine.Pawn|GetPlayerState Pawn)))
          (return (Class|FortPlayerStateAthena|IsAnAthenaGameParticipant _asfort_player_state_athena)))))
    (:then_1
      (return false))))

(fn IsActive
  (return (Math|Boolean|ANDBoolean (or (CallFunction|HasItems) (CallFunction|HasChannelSet)) (not (Variables|Default|GetbDisabledSpawning)) (or (VisibleInGameComponent|IsVisibleDuringMinigame) (not (Class|CreativeCommonDeviceFunctionLibrary|IsinPlayMode self))))))

(fn IsGameInProgress
  (bind _fortminigameprogress (Variables|Default|GetFortMinigameProgress)
    (:then
      (bind _currentminigame (Variables|/Script/FortniteGame.FortMinigameLogicComponent|GetCurrentMinigame _fortminigameprogress)
        (:then
          (return (Minigame|HasStarted _currentminigame :!class /Script/FortniteGame.FortMinigame)))
        (:else)))
    (:else)))

(fn OnPawnDeath
  (params (KilledPawn :type Fort%20Player%20Pawn%20Object%20Reference) (Damage :type Float%20%28double-precision%29) (InTags :type Gameplay%20Tag%20Container%20Structure) (EffectContext :type Gameplay%20Effect%20Context%20Handle%20Structure) (EventInstigator :type Controller%20Object%20Reference) (DamageCauser :type Actor%20Object%20Reference) (DBNOFinisher :type Controller%20Object%20Reference))
  (bind _asfort_player_pawn_athena (Utilities|Casting|CastToFortPlayerPawnAthena KilledPawn))
  (bind _finisherordownerpawn (Default|GetFinisherorDownerPawn _asfort_player_pawn_athena))
  (CallFunction|HandlePlayerDied _finisherordownerpawn _asfort_player_pawn_athena)
  (return))

(fn PassesSelfEliminationTest
  (params (DyingPlayer :type Fort%20Player%20Pawn%20Athena%20Object%20Reference%20%28by%20ref%29))
  (bind _dyingplayer (Variables|GetDyingPlayer))
  (if (CallFunction|ShouldSpawnonSelfElimination)
    (return true)
    (else
      (Utilities|FlowControl|Sequence
        (:then_0
          ;[1 1 1] Check for non-self-elimination
          (Utilities|IsValid _dyingplayer
            (:Is%20Valid
              (bind _playerstate (Variables|/Script/Engine.Pawn|GetPlayerState _dyingplayer)
                (:then
                  (bind _asfort_player_state_athena (Utilities|Casting|CastToFortPlayerStateAthena _playerstate))
                  (bind _returnvalue_finisherordowner (Utilities|Struct|BreakDeathInfo (Class|FortPlayerStateAthena|GetDeathInfo _asfort_player_state_athena)))
                  (return (Utilities|NotEqual(Object) _returnvalue_finisherordowner _asfort_player_state_athena)))
                (:else)))))
        (:then_1
          (return false))))))

(fn QueuePlayerLocation
  (params (PlayerPawnToQueue :type Actor%20Object%20Reference%20%28by%20ref%29))
  (Utilities|IsValid PlayerPawnToQueue
    (:Is%20Valid
      (if (not (Variables|Default|GetbDisabledSpawning))
        (bind _bshoulddropitems (CallFunction|ShouldRandomDropItems))
        (if _bshoulddropitems
          (Utilities|Array|Add (Variables|Default|GetQueuedSpawnLocations) (Transformation|GetActorLocation (Variables|GetPlayerPawnToQueue)))
          (bind _minigame_spawner_component (Variables|Default|GetMinigame_Spawner_Component)
            (:then
              (if (not (Class|MinigameSpawnerComponent|IsSpawning _minigame_spawner_component))
                (Class|MinigameSpawnerComponent|ResetSpawner (Variables|Default|GetMinigame_Spawner_Component) false)
                (Class|MinigameSpawnerComponent|SpawnRemaining (Variables|Default|GetMinigame_Spawner_Component) true)))
            (:else)))))))

(fn Reset
  (Utilities|FlowControl|Sequence
    (:then_0
      (Utilities|Array|Clear (Variables|Default|GetQueuedSpawnLocations))
      (bind _minigame_spawner_component (Variables|Default|GetMinigame_Spawner_Component)
        (:then
          (Class|MinigameSpawnerComponent|ResetSpawner _minigame_spawner_component true))
        (:else)))
    (:then_1
      (CallFunction|CleanUpPlayers)
      (CallFunction|CleanUpMinigameDelegates))))

(fn SetShouldTossPickups
  (Variables|Default|GetMinigame_Spawner_Component
    (:then
      (switch int (Variables|Default|GetTossPickups)
        (:Default
          _)
        (:0
          _)
        (:1
          (Variables|/Game/Athena/Items/Traps/MinigameSpawner/Minigame_Spawner_Component.Minigame_Spawner_Component_C|SetbShouldTossPickups false (Variables|Default|GetMinigame_Spawner_Component))))
      (Variables|/Game/Athena/Items/Traps/MinigameSpawner/Minigame_Spawner_Component.Minigame_Spawner_Component_C|SetbShouldTossPickups true (Variables|Default|GetMinigame_Spawner_Component)))
    (:else)))

(fn SetupScreenFX
  (Networking|CanExecuteCosmeticEvents
    (:True
      (CallFunction|UpdateScreenColor)
      (CallFunction|UpdateSpawnerFX))))

(fn ShouldRandomDropItems
  (LocalVariables|SetRandomRoll (Math|Random|RandomIntegerinRange 0 99))
  (Math|Integer|CompareInt (LocalVariables|GetRandomRoll) (Variables|Default|GetDropChance)
    (:>
      _)
    (:==
      _)
    (:<
      (return true)))
  (return false))

(fn ShouldSpawnOnSelfElimination
  (switch int (Variables|Default|GetSpawnOnSelfElimination)
    (:Default
      _)
    (:0
      (return true))
    (:1
      _))
  (return false))

(fn UpdateScreenColor)

(fn ConstructionScript
  (|CallParentFunction)
  (Rendering|Material|SetCustomPrimitiveDataFloat (Variables|Debug|GetStaticMeshComponent) 27 17.0))
