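; Fort controller component for the Creative Elimination Manager device that relays player pawn deaths.
; Its OnPawnDied override runs the parent logic, then calls the OnPawnDeath dispatcher with the killed pawn, damage, tags, effect context, instigator, damage causer and DBNO finisher.
; BeginPlay and Tick are empty, and the component does no filtering or elimination handling of its own.
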(event EventBeginPlay)

(event EventTick
  (params (DeltaSeconds :type Float%20%28single-precision%29)))

(fn OnPawnDied
  (params (KilledPawn :type Fort%20Player%20Pawn%20Object%20Reference) (Damage :type Float%20%28single-precision%29) (InTags :type Gameplay%20Tag%20Container%20Structure%20%28by%20ref%29) (EffectContext :type Gameplay%20Effect%20Context%20Handle%20Structure%20%28by%20ref%29) (EventInstigator :type Controller%20Object%20Reference) (DamageCauser :type Actor%20Object%20Reference) (DBNOFinisher :type Controller%20Object%20Reference))
  (|CallParentFunction KilledPawn Damage InTags EffectContext EventInstigator DamageCauser DBNOFinisher)
  (Default|CallOnPawnDeath (Variables|GetKilledPawn) (Variables|GetDamage) (Variables|GetInTags) (Variables|GetEffectContext) (Variables|GetEventInstigator) (Variables|GetDamageCauser) (Variables|GetDBNOFinisher)))
